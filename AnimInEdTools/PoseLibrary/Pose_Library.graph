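# Editor utility widget with Copy and Paste buttons that act as an in-memory pose clipboard for Control Rigs in Sequencer.
# Copy stores world transforms of selected controls at the current frame in a per-rig map of name to transform.
# Paste writes those transforms back to the same rigs at the current frame of the focused Level Sequence, setting keys.
# Despite its name, Bone Transform Map Local holds world-space control transforms.
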EditorUtilityWidgetBlueprint /AnimInEdTools/PoseLibrary/Pose_Library
graphs [EventGraph {nodes [
   "On Clicked (Copy_Button)"() {is=title K2Node_ComponentBoundEvent}
   > Clear!(
    execute
    TargetMap:Map of Control Rig Object References to Struct Name Transform Map Structures=Get Control Rig Proxy Binding() -> (Control Rig Proxy Binding:Map of Control Rig Object References to Struct Name Transform Map Structures)
   ) -> (then) {is=title K2Node_CallFunction}
   > n1 = Get Current Time!(execute, TimeUnit:EMovieSceneTimeUnit Enum=DisplayRate) -> (
    then
    ReturnValue_Frame_FrameNumber:Frame Number Structure
    ReturnValue_Frame_SubFrame:float=~
    ReturnValue_Time:float=~
    ReturnValue_MarkedFrame:String
    ReturnValue_Timecode:Timecode Structure
    ReturnValue_PositionType:EMovieScenePositionType Enum
    ReturnValue_UpdateMethod:EUpdatePositionMethod Enum
    ReturnValue_bHasJumped:bool=~
   ) {is=title K2Node_CallFunction}
   > n2 = GetFocusedLevelSequence!(execute) -> (then, ReturnValue:Level Sequence Object Reference)
   > n3 = Set Focused Level Sequence(Focused Level Sequence:Level Sequence Object Reference=@n2) -> (Output_Get:Level Sequence Object Reference)
   > n4 = GetControlRigs!(execute, LevelSequence:Level Sequence Object Reference=@n3) -> (then, ReturnValue:Array of Control Rig Sequencer Binding Proxy Structures)
   > n5 = For Each Loop(Array:Array of Control Rig Sequencer Binding Proxy Structures=@n4) -> (Array Element:Control Rig Sequencer Binding Proxy Structure)
   > Clear(TargetMap:Map of Names to Transforms=Get Bone Transform Map Local() -> (Bone Transform Map Local:Map of Names to Transforms))
   n6 = Break Control Rig Sequencer Binding Proxy!(ControlRigSequencerBindingProxy:"Control Rig Sequencer Binding Proxy Structure (by ref)"=@n5) -> (
    Proxy:Movie Scene Binding Proxy Structure
    ControlRig:Control Rig Object Reference
    Track:Movie Scene Control Rig Parameter Track Object Reference
   ) {is=title K2Node_BreakStruct}
   > n7 = For Each Loop!(
    Exec
    Array:Array of Rig Element Key Structures=Get Controls!(
     self:Rig Hierarchy Object Reference=GetHierarchy!(self:Control Rig Object Reference=@n6.ControlRig) -> (ReturnValue:Rig Hierarchy Object Reference)
     bTraverse:bool=true
    ) -> (ReturnValue:Array of Rig Element Key Structures) {is=title K2Node_CallFunction}
   ) -> (LoopBody, Array Element_Type:ERigElementType Enum, Array Element_Name:Name, Array Index, Completed)
   Add!(
    execute=@n7.Completed
    TargetMap:Map of Control Rig Object References to Struct Name Transform Map Structures=Get Control Rig Proxy Binding()
    Key:"Control Rig Object Reference (by ref)"=@n6.ControlRig
    Value:"Struct Name Transform Map Structure (by ref)"=Make Struct Name Transform Map!(NameTransform_4_#1:Map of Names to Transforms=Get Bone Transform Map Local()) -> (Struct_NameTransformMap:Struct Name Transform Map Structure) {is=title K2Node_MakeStruct}
   ) -> (then) {is=title K2Node_CallFunction}
   n8 = IsControlSelected!(self:Control Rig Object Reference=@n6.ControlRig, InControlName:"Name (by ref)"=@n7.Array Element_Name) -> (ReturnValue:bool)
   Branch(execute=@n7.LoopBody, Condition=@n8)
   > n9 = GetControlRigWorldTransform!(
    execute
    LevelSequence:Level Sequence Object Reference=Get Focused Level Sequence() -> (Focused Level Sequence:Level Sequence Object Reference)
    ControlRig:Control Rig Object Reference=@n6.ControlRig
    ControlName:Name=@n7.Array Element_Name
    Frame:Frame Number Structure=@n1
    TimeUnit:EMovieSceneTimeUnit Enum=DisplayRate
   ) -> (then, ReturnValue:Transform)
   > Add(
    TargetMap:Map of Names to Transforms=Get Bone Transform Map Local()
    Key:"Name (by ref)"=@n7.Array Element_Name
    Value:"Transform (by ref)"=@n9
   )
   "On Clicked (Paste_Button)"() {is=title K2Node_ComponentBoundEvent}
   > n10 = Get Current Time(TimeUnit=DisplayRate) -> (ReturnValue_Frame_SubFrame=~, ReturnValue_Time=~, ReturnValue_bHasJumped=~)
   > n11 = GetFocusedLevelSequence()
   > n12 = Keys!(execute, TargetMap:Map of Control Rig Object References to Struct Name Transform Map Structures=Get Control Rig Proxy Binding()) -> (then, Keys:Array of Control Rig Object References)
   > n13 = For Each Loop!(Exec, Array:Array of Control Rig Object References=@n12) -> (LoopBody, Array Element:Control Rig Object Reference, Array Index, Completed)
   > Is Valid(InputObject=@n13)
   n14 = Find!(
    TargetMap:Map of Control Rig Object References to Struct Name Transform Map Structures=Get Control Rig Proxy Binding()
    Key:"Control Rig Object Reference (by ref)"=@n13
   ) -> (Value_NameTransform_4_#1:Map of Names to Transforms, ReturnValue)
   > Branch(Condition=@n14.ReturnValue)
   > n15 = Keys(TargetMap:Map of Names to Transforms=@n14) -> (Keys:Array of Names)
   > n16 = For Each Loop(Array:Array of Names=@n15) -> (Array Element:Name)
   > SetControlRigWorldTransform!(
    execute
    LevelSequence:Level Sequence Object Reference=@n11
    ControlRig:Control Rig Object Reference=@n13
    ControlName:Name=@n16
    Frame:Frame Number Structure=@n10
    WorldTransform:"Transform (by ref)"=Find!(TargetMap:Map of Names to Transforms=@n14, Key:"Name (by ref)"=@n16) -> (Value:Transform, ReturnValue)
    TimeUnit:EMovieSceneTimeUnit Enum=DisplayRate
    bSetKey:bool=true
   ) -> (then)
   AND Boolean(A=EndsWith(SourceString="To String (Name)"(InName=@n7.Array Element_Name), InSuffix=_ctrl), B=@n8)
   Event PreConstruct()
   Event Construct()
   Event Tick()
  ]}]
variables [
 Bone Transform Map Local {type="map<Transform>", default {}}
 Control Rig Proxy Binding {default {}}
 Control Transform Map {type="map<Transform>", default {}}
 Focused Level Sequence {type=LevelSequence, default=null}
]
widgets {
 info {rootWidgetClass=&HorizontalBox, widgetCount=5}
 widgets [
  HorizontalBox_0 {widgetClassPath=&HorizontalBox, widget=HorizontalBox_0}
  Copy_Button {
   widgetClassPath=&Button
   parent=HorizontalBox_0
   widget=Copy_Button
   properties {widgetStyle {
     clickedSlateSound {resourceObject=null}
     disabled {
      bIsDynamicallyLoaded=false
      drawAs=RoundedBox
      imageType=NoImage
      mirroring=NoMirror
      resourceName=None
      resourceObject=null
      tiling=NoTile
      imageSize=<0 0>
      margin=<ltrb 0 0 0 0>
      outlineSettings {
       bUseBrushTransparency=false
       roundingType=FixedRadius
       width=1
       color {colorUseRule=UseColor_ColorTable, specifiedColor=<rgba 0 0 0 0>}
       cornerRadii=<4 4 4 4>
      }
      tintColor {colorUseRule=UseColor_Specified, specifiedColor=<rgba 0.0395 0.0395 0.0395 1>}
      uVRegion {bIsValid=false, max=<0 0>, min=<0 0>}
     }
     disabledForeground {colorUseRule=UseColor_ColorTable, specifiedColor=<rgba 0 0 0 0>}
     hovered {
      bIsDynamicallyLoaded=false
      drawAs=RoundedBox
      imageType=NoImage
      mirroring=NoMirror
      resourceName=None
      resourceObject=null
      tiling=NoTile
      imageSize=<32 32>
      margin=<ltrb 0 0 0 0>
      outlineSettings {
       bUseBrushTransparency=false
       roundingType=FixedRadius
       width=1
       color {colorUseRule=UseColor_ColorTable, specifiedColor=<rgba 0 0 0 0>}
       cornerRadii=<4 4 4 4>
      }
      tintColor {colorUseRule=UseColor_Specified, specifiedColor=<rgba 0.0953 0.0953 0.0953 1>}
      uVRegion {bIsValid=false, max=<0 0>, min=<0 0>}
     }
     hoveredForeground {colorUseRule=UseColor_ColorTable, specifiedColor=<rgba 0 0 0 0>}
     hoveredSlateSound {resourceObject=null}
     normal {
      bIsDynamicallyLoaded=false
      drawAs=RoundedBox
      imageType=NoImage
      mirroring=NoMirror
      resourceName=None
      resourceObject=null
      tiling=NoTile
      imageSize=<32 32>
      margin=<ltrb 0 0 0 0>
      outlineSettings {
       bUseBrushTransparency=false
       roundingType=FixedRadius
       width=1
       color {colorUseRule=UseColor_ColorTable, specifiedColor=<rgba 0 0 0 0>}
       cornerRadii=<4 4 4 4>
      }
      tintColor {colorUseRule=UseColor_Specified, specifiedColor=<rgba 0.0395 0.0395 0.0395 1>}
      uVRegion {bIsValid=false, max=<0 0>, min=<0 0>}
     }
     normalForeground {colorUseRule=UseColor_ColorTable, specifiedColor=<rgba 0 0 0 0>}
     normalPadding=<ltrb 12 1.5 12 1.5>
     pressed {
      bIsDynamicallyLoaded=false
      drawAs=RoundedBox
      imageType=NoImage
      mirroring=NoMirror
      resourceName=None
      resourceObject=null
      tiling=NoTile
      imageSize=<32 32>
      margin=<ltrb 0 0 0 0>
      outlineSettings {
       bUseBrushTransparency=false
       roundingType=FixedRadius
       width=1
       color {colorUseRule=UseColor_ColorTable, specifiedColor=<rgba 0 0 0 0>}
       cornerRadii=<4 4 4 4>
      }
      tintColor {colorUseRule=UseColor_Specified, specifiedColor=<rgba 0.0284 0.0284 0.0284 1>}
      uVRegion {bIsValid=false, max=<0 0>, min=<0 0>}
     }
     pressedForeground {colorUseRule=UseColor_ColorTable, specifiedColor=<rgba 0 0 0 0>}
     pressedPadding=<ltrb 12 2.5 12 0.5>
     pressedSlateSound {resourceObject=null}
    }}
   slot {class=HorizontalBoxSlot, properties {verticalAlignment=VAlign_Top, size {sizeRule=Fill, value=1}}}
  }
  Edit_2 {
   widgetClassPath=&TextBlock
   parent=Copy_Button
   widget=Edit_2
   properties {
    autoWrapText=true
    justification=Center
    text=Copy
    colorAndOpacity {colorUseRule=UseColor_Specified, specifiedColor=<rgba 0.5255 0.5255 0.5255 1>}
    font {
     bForceMonospaced=false
     bMaterialIsStencil=false
     fontMaterial=null
     fontObject=&Roboto
     letterSpacing=0
     monospacedWidth=1
     size=10
     skewAmount=0
     typefaceFontName=Bold
     outlineSettings {
      bApplyOutlineToDropShadows=false
      bMiteredCorners=false
      bSeparateFillAlpha=false
      outlineMaterial=null
      outlineSize=0
      outlineColor=<rgba 0 0 0 1>
     }
    }
   }
   slot {class=ButtonSlot}
  }
  Paste_Button {
   widgetClassPath=&Button
   parent=HorizontalBox_0
   widget=Paste_Button
   properties {widgetStyle {
     clickedSlateSound {resourceObject=null}
     disabled {
      bIsDynamicallyLoaded=false
      drawAs=RoundedBox
      imageType=NoImage
      mirroring=NoMirror
      resourceName=None
      resourceObject=null
      tiling=NoTile
      imageSize=<0 0>
      margin=<ltrb 0 0 0 0>
      outlineSettings {
       bUseBrushTransparency=false
       roundingType=FixedRadius
       width=1
       color {colorUseRule=UseColor_ColorTable, specifiedColor=<rgba 0 0 0 0>}
       cornerRadii=<4 4 4 4>
      }
      tintColor {colorUseRule=UseColor_Specified, specifiedColor=<rgba 0.0395 0.0395 0.0395 1>}
      uVRegion {bIsValid=false, max=<0 0>, min=<0 0>}
     }
     disabledForeground {colorUseRule=UseColor_ColorTable, specifiedColor=<rgba 0 0 0 0>}
     hovered {
      bIsDynamicallyLoaded=false
      drawAs=RoundedBox
      imageType=NoImage
      mirroring=NoMirror
      resourceName=None
      resourceObject=null
      tiling=NoTile
      imageSize=<32 32>
      margin=<ltrb 0 0 0 0>
      outlineSettings {
       bUseBrushTransparency=false
       roundingType=FixedRadius
       width=1
       color {colorUseRule=UseColor_ColorTable, specifiedColor=<rgba 0 0 0 0>}
       cornerRadii=<4 4 4 4>
      }
      tintColor {colorUseRule=UseColor_Specified, specifiedColor=<rgba 0.0953 0.0953 0.0953 1>}
      uVRegion {bIsValid=false, max=<0 0>, min=<0 0>}
     }
     hoveredForeground {colorUseRule=UseColor_ColorTable, specifiedColor=<rgba 0 0 0 0>}
     hoveredSlateSound {resourceObject=null}
     normal {
      bIsDynamicallyLoaded=false
      drawAs=RoundedBox
      imageType=NoImage
      mirroring=NoMirror
      resourceName=None
      resourceObject=null
      tiling=NoTile
      imageSize=<32 32>
      margin=<ltrb 0 0 0 0>
      outlineSettings {
       bUseBrushTransparency=false
       roundingType=FixedRadius
       width=1
       color {colorUseRule=UseColor_ColorTable, specifiedColor=<rgba 0 0 0 0>}
       cornerRadii=<4 4 4 4>
      }
      tintColor {colorUseRule=UseColor_Specified, specifiedColor=<rgba 0.0395 0.0395 0.0395 1>}
      uVRegion {bIsValid=false, max=<0 0>, min=<0 0>}
     }
     normalForeground {colorUseRule=UseColor_ColorTable, specifiedColor=<rgba 0 0 0 0>}
     normalPadding=<ltrb 12 1.5 12 1.5>
     pressed {
      bIsDynamicallyLoaded=false
      drawAs=RoundedBox
      imageType=NoImage
      mirroring=NoMirror
      resourceName=None
      resourceObject=null
      tiling=NoTile
      imageSize=<32 32>
      margin=<ltrb 0 0 0 0>
      outlineSettings {
       bUseBrushTransparency=false
       roundingType=FixedRadius
       width=1
       color {colorUseRule=UseColor_ColorTable, specifiedColor=<rgba 0 0 0 0>}
       cornerRadii=<4 4 4 4>
      }
      tintColor {colorUseRule=UseColor_Specified, specifiedColor=<rgba 0.0284 0.0284 0.0284 1>}
      uVRegion {bIsValid=false, max=<0 0>, min=<0 0>}
     }
     pressedForeground {colorUseRule=UseColor_ColorTable, specifiedColor=<rgba 0 0 0 0>}
     pressedPadding=<ltrb 12 2.5 12 0.5>
     pressedSlateSound {resourceObject=null}
    }}
   slot {class=HorizontalBoxSlot, properties {verticalAlignment=VAlign_Top, size {sizeRule=Fill, value=1}}}
  }
  Edit {
   widgetClassPath=&TextBlock
   parent=Paste_Button
   widget=Edit
   properties {
    autoWrapText=true
    justification=Center
    text=Paste
    colorAndOpacity {colorUseRule=UseColor_Specified, specifiedColor=<rgba 0.5255 0.5255 0.5255 1>}
    font {
     bForceMonospaced=false
     bMaterialIsStencil=false
     fontMaterial=null
     fontObject=&Roboto
     letterSpacing=0
     monospacedWidth=1
     size=10
     skewAmount=0
     typefaceFontName=Bold
     outlineSettings {
      bApplyOutlineToDropShadows=false
      bMiteredCorners=false
      bSeparateFillAlpha=false
      outlineMaterial=null
      outlineSize=0
      outlineColor=<rgba 0 0 0 1>
     }
    }
   }
   slot {class=ButtonSlot}
  }
 ]
}
